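; Gameplay volume actor for fire regions that shows its area through an overlap mesh and owns a playspace.
; The mesh gets a dynamic inactive or active fire material depending on whether ignition is prevented.
; A visibility update toggles the mesh and hides the actor in the editor unless it is visible in game, so Game Mode view works correctly.
; Destroying the actor also destroys its playspace.
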(event Custom|Update%20Visibility
  (params (CurrentVisibility :type Boolean) (VisibleInGame :type Boolean))
  (Rendering|SetVisibility (Variables|Default|GetOverlapMesh) CurrentVisibility false)
  (Rendering|SetActorHiddenInGame (and (not VisibleInGame) (Editor|InEditor))) ; If we're the editor and don't want to show this in game,\nwe always want it hidden (so Game Mode (g) works correctly). Else we always want it visible.
)

(event EventEndPlay
  (params (EndPlayReason :type EEndPlayReason%20Enum :default "Destroyed"))
  (if (Utilities|Enum|Equal(Enum) EndPlayReason "Destroyed")
    (Class|PlayspaceLibrary|DestroyPlayspace (CallFunction|GetPlayspace))))

(event Custom|Set%20Mesh%20Material
  (params (Prevent%20Ignition :type Boolean))
  (bind _returnvalue (Rendering|Material|CreateDynamicMaterialInstance (Selection|SelectMaterialInstance (Variables|Default|GetFireInactive) (Variables|Default|GetFireActive) Prevent%20Ignition) "None" "Transient" :!class /Script/Engine.KismetMaterialLibrary))
  (Rendering|Material|SetMaterial (Variables|Default|GetOverlapMesh) 0 _returnvalue))

(fn ConstructionScript)
